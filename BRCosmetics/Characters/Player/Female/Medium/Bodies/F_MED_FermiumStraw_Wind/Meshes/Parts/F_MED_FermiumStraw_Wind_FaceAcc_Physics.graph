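# Physics asset for the face accessory mesh of the female medium FermiumStraw_Wind player body, setting up dynamic hair bones around the head.
# Capsule bodies cover 29 Dyn_hair bones plus one large head capsule.
# Constraints link the hair into eleven chains, four of two bones and seven of three, each rooted to the head at its highest-numbered bone.
# Tight swing limits and near-locked twist restrict each link to slight sway.
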PhysicsAsset /BRCosmetics/Characters/Player/Female/Medium/Bodies/F_MED_FermiumStraw_Wind/Meshes/Parts/F_MED_FermiumStraw_Wind_FaceAcc_Physics
bodies {
 Dyn_hair_01 [{shapeType=Capsule, length=3.5326, radius=1.6504, center=<1 0 0>, rotation=<rot 90 1.6138 0>}]
 Dyn_hair_02 [{shapeType=Capsule, length=3.5326, radius=1.6504, center=<1 0 0>, rotation=<rot 90 0 0>}]
 Dyn_hair_03 [{shapeType=Capsule, length=2.668, radius=1.6601, center=<1 0 0>, rotation=<rot 90 0 0>}]
 Dyn_hair_04 [{shapeType=Capsule, length=2.668, radius=1.6601, center=<1 0 0>, rotation=<rot 90 0 0>}]
 Dyn_hair_05 [{shapeType=Capsule, length=2.6174, radius=1.527, center=<1 0 0>, rotation=<rot 90 27.7757 0>}]
 Dyn_hair_06 [{shapeType=Capsule, length=2.5803, radius=1.564, center=<1 0 0>, rotation=<rot 90 0 0>}]
 Dyn_hair_07 [{shapeType=Capsule, length=4.0268, radius=1, center=<1 0 0>, rotation=<rot 90 0 0>}]
 Dyn_hair_08 [{shapeType=Capsule, length=4.0268, radius=1, center=<1 0 0>, rotation=<rot 87.4924 -90.0001 -90.0001>}]
 Dyn_hair_09 [{shapeType=Capsule, length=3.619, radius=1.7742, center=<1 0 0>, rotation=<rot 90 0 0>}]
 Dyn_hair_10 [{shapeType=Capsule, length=3.619, radius=1.7742, center=<1 0 0>, rotation=<rot 90 0 0>}]
 Dyn_hair_11 [{shapeType=Capsule, length=3.619, radius=1.7742, center=<1 0 0>, rotation=<rot 90 0 0>}]
 Dyn_hair_12 [{shapeType=Capsule, length=3.2871, radius=1.962, center=<1 0 0>, rotation=<rot 90 0 0>}]
 Dyn_hair_13 [{shapeType=Capsule, length=3.2871, radius=1.962, center=<1 0 0>, rotation=<rot 90 0 0>}]
 Dyn_hair_14 [{shapeType=Capsule, length=3.2871, radius=1.962, center=<1 0 0>, rotation=<rot 90 0 0>}]
 Dyn_hair_15 [{shapeType=Capsule, length=3.007, radius=1.6459, center=<1 0 0>, rotation=<rot 90 0 0>}]
 Dyn_hair_16 [{shapeType=Capsule, length=3.007, radius=1.6459, center=<1 0 0>, rotation=<rot 90 0 0>}]
 Dyn_hair_17 [{shapeType=Capsule, length=3.007, radius=1.6459, center=<1 0 0>, rotation=<rot 90 0 0>}]
 Dyn_hair_18 [{shapeType=Capsule, length=3.4935, radius=2.2602, center=<1 0 0>, rotation=<rot 90 -26.5899 0>}]
 Dyn_hair_19 [{shapeType=Capsule, length=1.8824, radius=2.2232, center=<1 0 0>, rotation=<rot 90 3.6747 0>}]
 Dyn_hair_20 [{shapeType=Capsule, length=3.8653, radius=1.8106, center=<1.7839 0 0>, rotation=<rot 90 0 0>}]
 Dyn_hair_21 [{shapeType=Capsule, length=2.0599, radius=1.9401, center=<1 0 0>, rotation=<rot 90 0 0>}]
 Dyn_hair_22 [{shapeType=Capsule, length=2.0599, radius=1.9401, center=<1 0 0>, rotation=<rot 90 0 0>}]
 Dyn_hair_23 [{shapeType=Capsule, length=2.0599, radius=1.9401, center=<1 0 0>, rotation=<rot 90 0 0>}]
 Dyn_hair_24 [{shapeType=Capsule, length=1.9851, radius=1.9396, center=<1 0 0>, rotation=<rot 90 0 0>}]
 Dyn_hair_25 [{shapeType=Capsule, length=1.9851, radius=1.9396, center=<1 0 0>, rotation=<rot 90 0 0>}]
 Dyn_hair_26 [{shapeType=Capsule, length=2.3247, radius=1.6, center=<1 0 0>, rotation=<rot 90 0 0>}]
 Dyn_hair_27 [{shapeType=Capsule, length=2.4317, radius=0.9579, center=<1 0 0>, rotation=<rot 90 0 0>}]
 Dyn_hair_28 [{shapeType=Capsule, length=1.7738, radius=2.1899, center=<1 0 0>, rotation=<rot 90 0 0>}]
 Dyn_hair_29 [{shapeType=Capsule, length=1.6783, radius=2.2854, center=<1 0 0>, rotation=<rot 90 0 0>}]
 head [{shapeType=Capsule, length=2.0289, radius=9.6668, center=<0 1.3068 0>, rotation=<rot 90 0 0>}]
}
constraints [
 {bone1Name=Dyn_hair_02, bone2Name=head, swing1LimitDegrees=10, swing2LimitDegrees=10, twistLimitDegrees=1}
 {bone1Name=Dyn_hair_01, bone2Name=Dyn_hair_02, swing1LimitDegrees=10, swing2LimitDegrees=10, twistLimitDegrees=1}
 {bone1Name=Dyn_hair_04, bone2Name=head, swing1LimitDegrees=5, swing2LimitDegrees=5, twistLimitDegrees=1}
 {bone1Name=Dyn_hair_03, bone2Name=Dyn_hair_04, swing1LimitDegrees=5, swing2LimitDegrees=6, twistLimitDegrees=1}
 {bone1Name=Dyn_hair_06, bone2Name=head, swing1LimitDegrees=2, swing2LimitDegrees=8, twistLimitDegrees=1}
 {bone1Name=Dyn_hair_05, bone2Name=Dyn_hair_06, swing1LimitDegrees=2, swing2LimitDegrees=2, twistLimitDegrees=0}
 {bone1Name=Dyn_hair_08, bone2Name=head, swing1LimitDegrees=6, swing2LimitDegrees=8, twistLimitDegrees=1}
 {bone1Name=Dyn_hair_07, bone2Name=Dyn_hair_08, swing1LimitDegrees=5, swing2LimitDegrees=5, twistLimitDegrees=1}
 {bone1Name=Dyn_hair_11, bone2Name=head, swing1LimitDegrees=10, swing2LimitDegrees=4, twistLimitDegrees=1}
 {bone1Name=Dyn_hair_10, bone2Name=Dyn_hair_11, swing1LimitDegrees=6, swing2LimitDegrees=4, twistLimitDegrees=1}
 {bone1Name=Dyn_hair_09, bone2Name=Dyn_hair_10, swing1LimitDegrees=10, swing2LimitDegrees=4, twistLimitDegrees=1}
 {bone1Name=Dyn_hair_14, bone2Name=head, swing1LimitDegrees=5, swing2LimitDegrees=4, twistLimitDegrees=1}
 {bone1Name=Dyn_hair_13, bone2Name=Dyn_hair_14, swing1LimitDegrees=5, swing2LimitDegrees=4, twistLimitDegrees=1}
 {bone1Name=Dyn_hair_12, bone2Name=Dyn_hair_13, swing1LimitDegrees=6, swing2LimitDegrees=4, twistLimitDegrees=1}
 {bone1Name=Dyn_hair_17, bone2Name=head, swing1LimitDegrees=5, swing2LimitDegrees=3, twistLimitDegrees=1}
 {bone1Name=Dyn_hair_16, bone2Name=Dyn_hair_17, swing1LimitDegrees=5, swing2LimitDegrees=3, twistLimitDegrees=1}
 {bone1Name=Dyn_hair_15, bone2Name=Dyn_hair_16, swing1LimitDegrees=10, swing2LimitDegrees=3, twistLimitDegrees=1}
 {bone1Name=Dyn_hair_20, bone2Name=head, swing1LimitDegrees=15, swing2LimitDegrees=4, twistLimitDegrees=1}
 {bone1Name=Dyn_hair_19, bone2Name=Dyn_hair_20, swing1LimitDegrees=15, swing2LimitDegrees=4, twistLimitDegrees=1}
 {bone1Name=Dyn_hair_18, bone2Name=Dyn_hair_19, swing1LimitDegrees=15, swing2LimitDegrees=4, twistLimitDegrees=1}
 {bone1Name=Dyn_hair_23, bone2Name=head, swing1LimitDegrees=15, swing2LimitDegrees=6, twistLimitDegrees=1}
 {bone1Name=Dyn_hair_22, bone2Name=Dyn_hair_23, swing1LimitDegrees=12, swing2LimitDegrees=5, twistLimitDegrees=1}
 {bone1Name=Dyn_hair_21, bone2Name=Dyn_hair_22, swing1LimitDegrees=10, swing2LimitDegrees=5, twistLimitDegrees=1}
 {bone1Name=Dyn_hair_26, bone2Name=head, swing1LimitDegrees=15, swing2LimitDegrees=6, twistLimitDegrees=1}
 {bone1Name=Dyn_hair_25, bone2Name=Dyn_hair_26, swing1LimitDegrees=12, swing2LimitDegrees=6, twistLimitDegrees=1}
 {bone1Name=Dyn_hair_24, bone2Name=Dyn_hair_25, swing1LimitDegrees=10, swing2LimitDegrees=6, twistLimitDegrees=1}
 {bone1Name=Dyn_hair_29, bone2Name=head, swing1LimitDegrees=15, swing2LimitDegrees=5, twistLimitDegrees=1}
 {bone1Name=Dyn_hair_28, bone2Name=Dyn_hair_29, swing1LimitDegrees=12, swing2LimitDegrees=6, twistLimitDegrees=1}
 {bone1Name=Dyn_hair_27, bone2Name=Dyn_hair_28, swing1LimitDegrees=10, swing2LimitDegrees=6, twistLimitDegrees=1}
]
