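# Animation Blueprint for the face accessory part of the female FermiumStraw Wind cosmetic body, adding secondary motion to its earrings.
# It copies the attached parent's pose, then runs sprung, angle-limited AnimDynamics on each earring bone.
# A RigidBody node simulates the part's physics asset in head-bone space, then Modify Bone nodes additively scale both earrings.
# Each update stores GravityOverride from a nativized parameter-asset function, though no visible graph connection reads it.
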AnimBlueprint /BRCosmetics/Characters/Player/Female/Medium/Bodies/F_MED_FermiumStraw_Wind/Meshes/Parts/F_MED_FermiumStraw_Wind_FaceAcc_AnimBP
target_skeleton=&/F_MED_FermiumStraw_Wind_FaceAcc_Skeleton
graphs [
 AnimGraph {nodes [
   n1 = Anim Dynamics - Bone: earring_r(ComponentPose=Local To Component(LocalPose=Copy Pose From Mesh() {bUseAttachedParent=true}), Alpha=1, ExternalForce="0,0,0") {
    angularDampingOverride=1
    angularSpringConstant=500
    bAngularSpring=true
    bOverrideAngularDamping=true
    bUseGravityOverride=true
    simulationSpace=RootRelative
    alphaBoolBlend {blendOption=Linear}
    boundBone {boneName=earring_r}
    componentAppliedLinearAccClamp=<5000 5000 5000>
    componentLinearAccScale=<1 1 1>
    componentLinearVelScale=<1 1 1>
    gravityOverride {z=-980}
    physicsBodyDefinitions [{
      boundBone {boneName=earring_r}
      boxExtents=<10 10 10>
      constraintSetup {angularLimitsMax=<2 20 17>, angularLimitsMin=<-2 -2 -17>, angularTarget=<1 0 0>}
      localJointOffset=<0.6 0 0>
     }]
    relativeSpaceBone {boneName=head}
   }
   n2 = Anim Dynamics - Bone: earring_l(ComponentPose=@n1, Alpha=1, ExternalForce="0,0,0") {
    angularDampingOverride=1
    angularSpringConstant=500
    bAngularSpring=true
    bOverrideAngularDamping=true
    bUseGravityOverride=true
    simulationSpace=RootRelative
    alphaBoolBlend {blendOption=Linear}
    boundBone {boneName=earring_l}
    componentAppliedLinearAccClamp=<5000 5000 5000>
    componentLinearAccScale=<1 1 1>
    componentLinearVelScale=<1 1 1>
    gravityOverride {z=-980}
    physicsBodyDefinitions [{
      boundBone {boneName=earring_l}
      boxExtents=<10 10 10>
      constraintSetup {angularLimitsMax=<2 20 17>, angularLimitsMin=<-2 -2 -17>, angularTarget=<1 0 0>}
      localJointOffset=<0.6 0 0>
     }]
    relativeSpaceBone {boneName=head}
   }
   n3 = RigidBody!(ComponentPose=@n2, Alpha, OverrideWorldGravity:Vector="0,0,0", ExternalForce) -> (Pose) {
    bClampLinearTranslationLimitToRefPose=true
    bOverrideWorldGravity=true
    cachedBoundsScale=1
    overridePhysicsAsset=&/F_MED_FermiumStraw_Wind_FaceAcc_Physics
    simulationSpace=BaseBoneSpace
    alphaBoolBlend {blendOption=Linear}
    baseBoneRef {boneName=head}
    overrideWorldGravity {z=-980}
    simSpaceSettings {
     maxAngularAcceleration=40
     maxAngularVelocity=10
     maxLinearAcceleration=2700
     maxLinearVelocity=700
     worldAlpha=0.45
     externalLinearDragV=<0.1 0.1 0.1>
    }
   }
   n4 = "Transform (Modify) Bone - Bone: earring_l"(ComponentPose=@n3, Alpha=1, Translation="0,0,0", Rotation="0,0,0", Scale="0.8,0.8,1") {scaleMode=BMM_Additive, scaleSpace=BCS_ParentBoneSpace, boneToModify {boneName=earring_l}, scale=<0.8 0.8>}
   n5 = "Transform (Modify) Bone - Bone: earring_r"(ComponentPose=@n4, Alpha=1, Translation="0,0,0", Rotation="0,0,0", Scale="0.8,0.8,1") {scaleMode=BMM_Additive, scaleSpace=BCS_ParentBoneSpace, boneToModify {boneName=earring_r}, scale=<0.8 0.8>}
   Output Pose(Result=Component To Local(ComponentPose=@n5)) {layerGroup=None}
  ]}
 EventGraph {nodes [
   Event BlueprintUpdateAnimation()
   > n1 = "Gravity Override State Function (Parameter Asset) (Nativized)"(
    CharacterPartAnimInstance=Self-Reference()
    GameTimeInSeconds=GetGameTimeInSeconds()
    Parameters=&/F_MED_FermiumStraw_Wind_Parameters
   )
   > Set GravityOverride(GravityOverride=@n1)
  ]}
]
variables [GravityOverride {type=Vector}]
